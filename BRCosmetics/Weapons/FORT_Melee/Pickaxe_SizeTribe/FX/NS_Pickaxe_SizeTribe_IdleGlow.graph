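# Niagara system for the idle glow on the SizeTribe pickaxe, drawing soft cyan sprites along the handle.
# A single Handle_Glows emitter bursts two sprites at fixed local offsets along Z, nudged by a camera offset, which persist until the emitter deactivates.
# Alpha fades in over emitter age and scales with a time-of-day value and a multiplier that is higher in the front end than in gameplay.
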NiagaraSystem /BRCosmetics/Weapons/FORT_Melee/Pickaxe_SizeTribe/FX/NS_Pickaxe_SizeTribe_IdleGlow
dependencies {
 usedDataInterfaces [&NiagaraDataInterfaceCurve]
 usedDynamicInputs [&FNTimeOfDayScalar]
 usedModules [
  &SystemState
  &self:SystemSpawnScript.SystemScriptSource.SystemScriptGraph.NiagaraNodeAssignment_1.SetVariables_#1
 ]
}
emitters {Handle_Glows {
  emitterUpdateScript [
   &EmitterState(
    Inactive Response=::NewEnumerator1
    Life Cycle Mode=::NewEnumerator1
    Loop Behavior=::NewEnumerator1
    Loop Duration Mode=::NewEnumerator1
    UseLoopDelay=false
   )
   &SpawnBurst_Instantaneous(Spawn Count=2, Use Spawn Probability=false, UseLoopCountLimit=false)
  ]
  particleSpawnScript [
   &InitializeParticle(
    Color=<rgba 0 0.7904 1 0.075>
    Position Offset=&MergeHelper_VectorSwitch(0=<0 0 -17.5>, 1=<0 0 25>, 2=?, 3=?, Index=&ReturnExecIndex(), Num=::NewEnumerator0)
    Position Offset Coordinate Space=::Local
    Sprite Size=<40 40>
    Sprite Size Mode=::NewEnumerator4
    UsePositionOffset=true
    Write Material Random=true
   )
   &CameraOffset(Camera Offset Amount=20)
  ]
  particleUpdateScript [
   &ParticleState(Kill Particles When Lifetime Has Elapsed=false, Let Infinitely Lived Particles Die When Emitter Deactivates=true)
   &ScaleColor(
    Scale Alpha=&FloatFromCurve(
     FloatCurve=DataInterface() {propertyValues {Curve {defaultValue=1, keys [{tangentMode=RCTM_User, time=0, value=0}, {tangentMode=RCTM_User, time=0.5, value=1}]}}}
     CurveIndex=$Emitter.Age
     Scale Curve=&Multiply_Float(A=&FN_FrontEndFloatModifier(Gameplay=2, Front End=3), B=$System.TOD)
    )
    Scale RGB=&VectorFromFloat(Value=1)
    ScaleA=true
    ScaleRGB=true
   )
   &DynamicMaterialParameters(
    Index 0 Param 1=5
    Module.Index0Float0Write=true
    Module.Index0Float1Write=false
    Module.Index0Float2Write=false
    Module.Index0Float3Write=false
   )
  ]
  rendererClasses [&NiagaraSpriteRendererProperties]
  renderers [{rendererClass=&NiagaraSpriteRendererProperties}]
 }}
summary {
 systemName=NS_Pickaxe_SizeTribe_IdleGlow
 emitters [{emitterName=Handle_Glows, rendererClasses [&NiagaraSpriteRendererProperties]}]
 userVariables [
  User.bIsFrontEnd {type=NiagaraBool}
  User.bIsFrontEndPreview {type=NiagaraBool}
 ]
}
